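; NitroComponent_Base is a base component that tracks the active Nitro state gameplay effect on its owner.
; On BeginPlay it queries the owner's Fort ability system component for active effects matching the Nitro state tags and caches a handle from the results.
; Two functions read that cached handle to return the effect's remaining and total duration.
; The lookup runs only at BeginPlay, with no refresh in this graph.
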(event EventBeginPlay
  (CallFunction|CacheNitroStateEffectHandle))

(fn CacheNitroStateEffectHandle
  (bind _returnvalue (Components|GetOwner))
  (bind _returnvalue_2 (Ability|GetFortAbilitySystemComponent _returnvalue))
  (Utilities|IsValid _returnvalue_2
    (:Is%20Valid
      (bind _returnvalue_3 (GameplayEffects|GetActiveEffectswithAllTags _returnvalue_2 (Variables|Default|GetNitroStateTags)))
      (if (Utilities|Array|IsNotEmpty _returnvalue_3)
        (Variables|Default|SetNitroStateEffectHandle (Utilities|Array|Get(acopy) _returnvalue_3))))))

(fn GetNitroStateRemainingDuration
  (bind _returnvalue (Ability|GameplayEffect|GetActiveGameplayEffectRemainingDuration (Variables|Default|GetNitroStateEffectHandle)))
  (return _returnvalue))

(fn GetNitroStateTotalDuration
  (bind _returnvalue (Ability|GameplayEffect|GetActiveGameplayEffectTotalDuration (Variables|Default|GetNitroStateEffectHandle)))
  (return _returnvalue))
