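; Vehicle component that drives the Nitro state, applying a vehicle mod and local looping cue, then undoing its effects on EndPlay.
; It adds fuel on activation and each Nitro re-application, and zeroes fuel consumption meanwhile.
; Seat enter/exit events grant or remove a weapon buff on pawns in weapon seats.
; The server ticks to replicate remaining effect duration, since vehicle GameplayEffects don't replicate, and OnRep broadcasts it for client FX.
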(event EventBeginPlay
  (|CallParentFunction)
  (Development|PrintString "Vehicle component is alive!")
  (Utilities|FlowControl|Sequence
    (:then_0
      ;[1 1 1] LIsten for pawns entering and exiting seats to check if they are in a seat with a weapon
      (bind _asfort_athena_vehicle (Utilities|Casting|CastToFortAthenaVehicle (Components|GetOwner)))
      (Vehicle|AssignOnPawnEnteredVehicleSeat _asfort_athena_vehicle (EventDispatchers|CreateEvent :!event Custom|OnPawnEnteredVehicleSeat_Event))
      (Vehicle|AssignOnPawnExitedVehicleSeat _asfort_athena_vehicle (EventDispatchers|CreateEvent :!event Custom|OnPawnExitedVehicleSeat_Event)))
    (:then_1
      ;[1 1 1] Add some fuel when Nitro first activates, then listen for an event that fires every time Nitro is applied so we can get a little bit of fuel even if we're already in the Nitro state
      (CallFunction|AddFuel)
      (Ability|Async|WaitGameplayEventToActor (Components|GetOwner) (Variables|Default|GetNitroAppliedEvent)
        (:EventReceived
          (CallFunction|AddFuel))))
    (:then_2
      (bind _asfort_vehicle_interface (Utilities|Casting|CastToFortVehicleInterface (Components|GetOwner)))
      (bind _output_get (Variables|Default|SetFuelComponentRef (Vehicle|Fuel|GetVehicleFuelComponent _asfort_vehicle_interface)))
      (Utilities|IsValid _output_get
        (:Is%20Valid
          (Class|FuelComponent|SetFuelUsageInfo (Variables|Default|GetFuelComponentRef) (struct :FuelPerSecondIdle 0.0 :FuelPerSecondDriving 0.0 :FuelPerSecondBoosting 0.0)))))
    (:then_3
      (CallFunction|HandleNitroStateMod true))
    (:then_4
      (CallFunction|AddNitroStateGameplayCue))))

(event Custom|AddFuel
  ;[1 1 1] Add Fuel, so Nitro can let empty vehicles continue to drive
  (bind _asfort_vehicle_interface (Utilities|Casting|CastToFortVehicleInterface (Components|GetOwner)))
  (Vehicle|Fuel|SetFuel _asfort_vehicle_interface (+ (Vehicle|Fuel|GetFuel _asfort_vehicle_interface) (ScalableFloat|GetValueatLevel (Variables|Default|GetRow_FuelAdd)))))

(event EventTick
  (params (DeltaSeconds :type Float%20%28single-precision%29))
  ;[1 1 1] Read remaining duration from the actual GE on the server, then replicate that value to clients. GE's don't replicate to clients on vehicles, so this is necessary for FX to hook into the duration
  (if (Networking|IsServer)
    (Variables|Default|SetRemainingDuration (CallFunction|GetNitroStateRemainingDuration))
    (else
      (Components|Tick|SetComponentTickEnabled false) ; Don't need to tick on the client
    )))

(event Custom|OnPawnEnteredVehicleSeat_Event
  (params (Vehicle :type Fort%20Vehicle%20Interface%20Interface%20%28by%20ref%29) (PlayerPawn :type Fort%20Pawn%20Object%20Reference) (SeatIndex :type Integer))
  ;[1 1 1] Seat Enter/Exit Handlers
  ;;[1 1 1] Entered Seat - check if seat has a weapon, apply the buff and save out the pawn to the array (just in case there's ever a world where we have multiple weapons on one car)
  (Utilities|IsValid (VehicleWeapon|GetSeatWeaponComponent Vehicle SeatIndex)
    (:Is%20Valid
      (CallFunction|AddWeaponBuff PlayerPawn)
      (Utilities|Array|Add (Variables|Default|GetWeaponUsers) PlayerPawn))))

(event Custom|OnPawnExitedVehicleSeat_Event
  (params (Vehicle :type Fort%20Vehicle%20Interface%20Interface%20%28by%20ref%29) (PlayerPawn :type Fort%20Pawn%20Object%20Reference) (SeatIndex :type Integer))
  ;[1 1 1] Seat Enter/Exit Handlers
  ;;[1 1 1] Exited Seat - check if the seat we just left had a weapon, remove the buff, and remove from array
  (Utilities|IsValid (VehicleWeapon|GetSeatWeaponComponent Vehicle SeatIndex)
    (:Is%20Valid
      (CallFunction|RemoveWeaponBuff PlayerPawn)
      (Utilities|Array|RemoveItem (Variables|Default|GetWeaponUsers) PlayerPawn))))

(event EventEndPlay
  (params (EndPlayReason :type EEndPlayReason%20Enum :default "Destroyed"))
  (bind _fuelcomponentref (Variables|Default|GetFuelComponentRef))
  (CallFunction|RemoveGrantedTags)
  (CallFunction|HandleNitroStateMod false)
  (CallFunction|RemoveNitroStateGameplayCue)
  ;[1 1 1] Pawns might still be in our weapon seats when Nitro ends, so remove the buff
  (for _array_element (Variables|Default|GetWeaponUsers)
    (CallFunction|RemoveWeaponBuff _array_element))
  (Utilities|IsValid _fuelcomponentref
    (:Is%20Valid
      (Class|FuelComponent|ReinitializeFuelUsageInfo _fuelcomponentref))))

(fn AddNitroStateGameplayCue
  (Networking|SwitchHasAuthority
    (:Authority
      _)
    (:Remote
      (bind _asfort_athena_vehicle (Utilities|Casting|CastToFortAthenaVehicle (Components|GetOwner)))
      (GameplayEventRouter|BroadcastEvent (GameplayEventRouter|GetGlobalEventRouter) (Utilities|Struct|MakeFortEventNitroVehicleStateChanged true) _asfort_athena_vehicle)))
  (Projectile|AddGameplayCue%28Looping%29Local (Components|GetOwner) (Variables|Default|GetNitroStateCueTag) (Ability|GameplayCue|MakeGameplayCueParameters 0.0)))

(fn AddWeaponBuff
  (params (Actor :type Actor%20Object%20Reference))
  (GameplayEffects|ApplyGameplayEffectToSelf :self (Ability|GetFortAbilitySystemComponent Actor) :GameplayEffectClass (Variables|Default|GetWeaponBuffGE) :Level 0.0))

(fn HandleNitroStateMod
  (params (Adding :type Boolean))
  (bind _asfort_vehicle_interface (Utilities|Casting|CastToFortVehicleInterface (Components|GetOwner)))
  (if Adding
    (Vehicle|Mod|ApplyVehicleMod _asfort_vehicle_interface (Variables|Default|GetVehicleModTag) 0)
    (else
      (Vehicle|Mod|RemoveVehicleMod _asfort_vehicle_interface (Variables|Default|GetVehicleModTag)))))

(fn OnRep_RemainingDuration
  ;[1 1 1] Since vehicle GE's do not replicate, we read and replicate the remaining duration and broadcast it for clients to hook into
  (Default|CallOnRemainingDurationChanged (Variables|Default|GetRemainingDuration)))

(fn RemoveGrantedTag
  (params (TagToCheck :type Gameplay%20Tag%20Structure))
  (bind _returnvalue (Components|GetOwner))
  (GameplayTags|HasMatchingGameplayTag(Message) _returnvalue TagToCheck)
  (Ability|GameplayEffect|RemoveLooseGameplayTags _returnvalue (GameplayTags|MakeGameplayTagContainerfromTag TagToCheck) true))

(fn RemoveGrantedTags
  (for _array_element (GameplayTags|BreakGameplayTagContainer (Variables|Default|GetGrantedTagsToRemove))
    (CallFunction|RemoveGrantedTag _array_element)))

(fn RemoveNitroStateGameplayCue
  (Networking|SwitchHasAuthority
    (:Authority
      _)
    (:Remote
      (bind _asfort_athena_vehicle (Utilities|Casting|CastToFortAthenaVehicle (Components|GetOwner)))
      (GameplayEventRouter|BroadcastEvent (GameplayEventRouter|GetGlobalEventRouter) (Utilities|Struct|MakeFortEventNitroVehicleStateChanged false) _asfort_athena_vehicle)))
  (Projectile|RemoveGameplayCue%28Looping%29Local (Components|GetOwner) (Variables|Default|GetNitroStateCueTag) (Ability|GameplayCue|MakeGameplayCueParameters 0.0)))

(fn RemoveWeaponBuff
  (params (Actor :type Actor%20Object%20Reference))
  (GameplayEffects|RemoveActiveEffectswithGrantedTags (Ability|GetFortAbilitySystemComponent Actor) (Variables|Default|GetWeaponBuffGrantedTags)))
